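# Niagara effect for the DustyWing companion's emote, emitting colored meshes and glowing sprites from its skeletal mesh bones, tinted from three colors via User.VariantColor.
# DeliSandwish_VFX spawns short-lived meshes, partitioned for mirrored offsets, that drift outward, grow and fade with animated material parameters.
# Minimal bursts two sprites with lights, pinned to bones and pulsed by User.NormalizedNotifyProgress, apparently fed by an animation notify.
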NiagaraSystem /CosmeticCompanions/Assets/Other/DustyWing/FX/NS_Companion_DustyWing_Emote
dependencies {
 usedDataInterfaces [
  &NiagaraDataInterfaceArrayColor
  &NiagaraDataInterfaceArrayFloat3
  &NiagaraDataInterfaceCurve
  &NiagaraDataInterfaceArrayFloat
  &NiagaraDataInterfaceSkeletalMesh
 ]
 usedDynamicInputs [&SimulationPosition, &UniformRangedFloat]
 usedModules [
  &self:SystemSpawnScript.SystemScriptSource.SystemScriptGraph.NiagaraNodeAssignment_0.SetVariables_#1
  &SystemState
 ]
}
emitters {
 DeliSandwish_VFX {
  emitterUpdateScript [
   &EmitterState()
   &SpawnBurst_Instantaneous(Spawn Count=2, Use Spawn Probability=false, UseLoopCountLimit=false)
   &SpawnRate(SpawnRate=10, Use Spawn Probability=false)
  ]
  particleSpawnScript [
   &PartitionParticles()
   &InitializeParticle(Lifetime=0.7, Mesh Scale Mode=::NewEnumerator3, Mesh Uniform Scale=0.25, UsePositionOffset=false, Write Material Random=true) {module=InitializeParticle001}
   &Color(Color=&SelectLinearColorFromArray(
     Array Sampling Mode=::NewEnumerator1
     Direct Array Index=$User.VariantColor
     Color Selection Array=DataInterface() {propertyValues {ColorData [<rgba 1 0.0163 0 0.8>, <rgba 0.0566 0.724 0.094 0.8>, <rgba 1 0 0.7 0.8>]}}
     Random Seed=?
    ))
   &DynamicMaterialParameters(
    Index 0 Param 1=2
    Index 0 Param 4=8
    Module.Param0WriteEnabled=false
    Module.Param1WriteEnabled=false
    Module.Param2WriteEnabled=true
    Module.Param3WriteEnabled=true
   ) {module=DynamicMaterialParameters001}
   &InitialMeshOrientation(
    Mesh Orientation Mode=::NewEnumerator2
    Orientation Vector=&SelectVectorFromArray(
     Direct Array Index=$Particles.PartitionParticles.Partition
     Vector Selection Array=DataInterface() {propertyValues {FloatData [<0.3 1 0>, <0.3 1 0>]}}
    )
    Use Rotation Vector=false
   )
  ]
  particleUpdateScript [
   &Drag(Use Linear Drag=true, Use Rotational Drag=false)
   &SolveForcesAndVelocity(Clamp Velocity=false, Limit Acceleration=false)
   &SkeletalMeshLocation(
    Bone / Socket Index=$Particles.PartitionParticles.Partition
    MaskSpawnGroup=false
    Mesh Sampling Type=::NewEnumerator5
    Sampled Position Offset=&MakeVector(
     X=&FloatFromCurve(
      FloatCurve=DataInterface() {propertyValues {Curve {
         defaultValue=1
         keys [{interpMode=RCIM_Cubic, leaveTangent=1.0679, tangentMode=RCTM_Break, time=0, value=0}, {tangentMode=RCTM_User, time=1, value=1}]
        }}}
      Scale Curve=12
     )
     Y=0
     Z=0
    )
    Skeletal Mesh=$System.Skeletal Mesh Reader
    Socket Sampling Mode=::NewEnumerator5
   ) {enabled=false}
   &SkeletalMeshLocation(
    Bone / Socket Index=$Particles.PartitionParticles.Partition
    Bone Sampling Mode=::NewEnumerator4
    MaskSpawnGroup=false
    Sampled Position Offset=&MakeVector(
     X=&Add_Float(
      A=&SelectFloatFromArray(Direct Array Index=$Particles.PartitionParticles.Partition, Float Selection Array=DataInterface() {propertyValues {FloatData [-2,2]}})
      B=&FloatFromCurve(
       FloatCurve=DataInterface() {propertyValues {Curve {
          defaultValue=1
          keys [
           {interpMode=RCIM_Cubic, leaveTangent=-0.2253, tangentMode=RCTM_Break, time=0, value=0}
           {arriveTangent=1.0984, tangentMode=RCTM_User, time=1, value=1}
          ]
         }}}
       Scale Curve=&SelectFloatFromArray(Direct Array Index=$Particles.PartitionParticles.Partition, Float Selection Array=DataInterface() {propertyValues {FloatData [-5,5]}})
      )
     )
     Y=&Add_Float(
      A=&SelectFloatFromArray(Direct Array Index=$Particles.PartitionParticles.Partition, Float Selection Array=DataInterface() {propertyValues {FloatData [-3,3]}})
      B=&FloatFromCurve(
       FloatCurve=DataInterface() {propertyValues {Curve {
          defaultValue=1
          keys [
           {interpMode=RCIM_Cubic, leaveTangent=-0.2253, tangentMode=RCTM_Break, time=0, value=0}
           {arriveTangent=1.0984, tangentMode=RCTM_User, time=1, value=1}
          ]
         }}}
       Scale Curve=&SelectFloatFromArray(Direct Array Index=$Particles.PartitionParticles.Partition, Float Selection Array=DataInterface() {propertyValues {FloatData [-15,15]}})
      )
     )
     Z=0
    )
    Skeletal Mesh=$System.Skeletal Mesh Reader
   ) {module=SkeletalMeshLocation001}
   /SetVariables(Particles.LightAlpha:NiagaraFloat=&FloatFromCurve(
     FloatCurve=DataInterface() {propertyValues {Curve {keys [
         <at 0 0>
         <at 0.1 1>
         {arriveTangent=-3.6605, interpMode=RCIM_Cubic, leaveTangent=-3.6605, tangentMode=RCTM_User, time=0.45, value=1}
         {arriveTangent=-3.1112, interpMode=RCIM_Cubic, leaveTangent=-3.1112, tangentMode=RCTM_User, time=1.0049, value=-1}
        ]}}}
     CurveIndex=$Emitter.NormalizedLoopAge
    )) {isSetParameters=true}
   &ParticleState()
   &ScaleMeshSize(Scale Factor=&VectorFromFloat(Value=&FloatFromCurve(FloatCurve=DataInterface() {propertyValues {Curve {keys [
          {arriveTangent=0.1305, interpMode=RCIM_Cubic, leaveTangent=0.1305, tangentMode=RCTM_User, time=0, value=1}
          {arriveTangent=0.1729, interpMode=RCIM_Cubic, leaveTangent=0.173, tangentMode=RCTM_User, time=1, value=2.2}
         ]}}})))
   &ScaleColor(
    Scale Alpha=&FloatFromCurve(
     FloatCurve=DataInterface() {propertyValues {Curve {keys [<at 0 0>, {interpMode=RCIM_Cubic, leaveTangent=-2.5686, tangentMode=RCTM_Break, time=0.1397, value=1}, <at 1 0>]}}}
     Scale Curve=$Particles.LightAlpha
    )
    ScaleA=true
    ScaleRGB=false
   )
   &DynamicMaterialParameters(
    Index 0 Param 1=&FloatFromCurve(
     FloatCurve=DataInterface() {propertyValues {Curve {keys [{arriveTangent=3.2486, interpMode=RCIM_Cubic, leaveTangent=3.2485, tangentMode=RCTM_User, time=0, value=0}, <at 1 2>]}}}
     Scale Curve=0.1
    )
    Index 0 Param 4=&FloatFromCurve(FloatCurve=DataInterface() {propertyValues {Curve {keys [<at 0 5>,<at 1 6>]}}}, Scale Curve=2)
    Module.Param0WriteEnabled=true
    Module.Param1WriteEnabled=false
    Module.Param2WriteEnabled=false
    Module.Param3WriteEnabled=true
   )
   &ScaleColor(
    Scale Alpha=&FloatFromCurve(FloatCurve=DataInterface() {propertyValues {Curve {keys [<at 0 1>,<at 0.7801 0.9765>,<at 1 0>]}}}, CurveIndex=$User.NormalizedNotifyProgress)
    ScaleA=true
    ScaleRGB=true
   ) {enabled=false, module=ScaleColor001}
  ]
  rendererClasses [&NiagaraMeshRendererProperties]
  renderers [{rendererClass=&NiagaraMeshRendererProperties, rendererIndex=0}]
 }
 Minimal {
  emitterUpdateScript [
   &EmitterState(Life Cycle Mode=::NewEnumerator1, Loop Behavior=::NewEnumerator1, Loop Duration Mode=::NewEnumerator1, UseLoopDelay=false)
   &SpawnBurst_Instantaneous(Spawn Count=2, Use Spawn Probability=false, UseLoopCountLimit=false)
  ]
  particleSpawnScript [&InitializeParticle(
    Color=&SelectLinearColorFromArray(
     Array Sampling Mode=::NewEnumerator1
     Direct Array Index=$User.VariantColor
     Color Selection Array=DataInterface() {propertyValues {ColorData [<rgba 1 0.0163 0 0.4>, <rgba 0.0566 0.724 0.094 0.4>, <rgba 1 0 0.7 0.4>]}}
     Random Seed=?
    )
    Lifetime=3
    Sprite Size Mode=::NewEnumerator3
    Uniform Sprite Size=25
    UsePositionOffset=false
    Write Material Random=true
   )]
  particleUpdateScript [
   &ParticleState()
   &CameraOffset(Camera Offset Amount=12)
   &SkeletalMeshLocation(
    Bone / Socket Index=&ReturnExecIndex()
    MaskSpawnGroup=false
    Mesh Sampling Type=::NewEnumerator5
    Skeletal Mesh=$System.Skeletal Mesh Reader
    Socket Sampling Mode=::NewEnumerator5
   ) {enabled=false}
   &SkeletalMeshLocation(
    Bone / Socket Index=&ReturnExecIndex()
    Bone Sampling Mode=::NewEnumerator4
    MaskSpawnGroup=false
    Skeletal Mesh=$System.Skeletal Mesh Reader
   ) {module=SkeletalMeshLocation001}
   &ScaleColor(
    Scale Alpha=&FloatFromCurve(
     FloatCurve=DataInterface() {propertyValues {Curve {keys [
         {interpMode=RCIM_Cubic, leaveTangent=12.6173, tangentMode=RCTM_Break, time=0.0059, value=0}
         {interpMode=RCIM_Cubic, tangentMode=RCTM_Break, time=0.1671, value=1}
         {interpMode=RCIM_Cubic, tangentMode=RCTM_Break, time=0.272, value=0.4239}
         {interpMode=RCIM_Cubic, tangentMode=RCTM_Break, time=0.3933, value=1}
         {interpMode=RCIM_Cubic, tangentMode=RCTM_Break, time=0.5091, value=0.4176}
         {interpMode=RCIM_Cubic, tangentMode=RCTM_Break, time=0.6227, value=1}
         {interpMode=RCIM_Cubic, tangentMode=RCTM_Break, time=0.7658, value=0.2693}
         {interpMode=RCIM_Cubic, tangentMode=RCTM_Break, time=0.8778, value=1}
         <at 1 0>
        ]}}}
     CurveIndex=$User.NormalizedNotifyProgress
    )
    ScaleA=true
    ScaleRGB=true
   )
   &FN_Light_Attributes(
    Color=$Particles.Color
    Create Color Attribute=true
    Create Exponent Attribute=true
    Create Radius Attribute=true
    Exponent=3
    Radius=20
    Scale Color=2
   )
  ]
  rendererClasses [&NiagaraSpriteRendererProperties, &NiagaraLightRendererProperties]
  renderers [
   {rendererClass=&NiagaraSpriteRendererProperties}
   {rendererClass=&NiagaraLightRendererProperties, rendererIndex=1}
  ]
 }
}
summary {
 systemName=NS_Companion_DustyWing_Emote
 emitters [
  {emitterName=DeliSandwish_VFX, rendererClasses [&NiagaraMeshRendererProperties]}
  {emitterName=Minimal, rendererClasses [&NiagaraSpriteRendererProperties, &NiagaraLightRendererProperties]}
 ]
 userVariables [
  User.EyeVariant {type=NiagaraFloat, defaultValue=0, description=""}
  User.NormalizedNotifyProgress {type=NiagaraFloat, defaultValue=0.5, description=""}
  User.VariantColor {type=NiagaraInt32, defaultValue=-1}
  User.VariantID {type=NiagaraFloat, defaultValue=1, description=""}
 ]
}
