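# Niagara idle effect for the DustyWing cosmetic companion, drawing sprites attached to its skeletal mesh through a shared mesh reader.
# BodySketches continuously spawns bone-sampled sprites, with size and offset chosen by VariantID and colour from Base Color tinted by time of day.
# EyeGlow bursts two non-expiring sprites, each pinned to a socket by execution index and tinted by EmissiveTint.
# Front-end modifiers adjust body sprite size and brightness.
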NiagaraSystem /CosmeticCompanions/Assets/Other/DustyWing/FX/NS_Companion_DustyWing_Idle
dependencies {
 usedDataInterfaces [&NiagaraDataInterfaceArrayFloat, &NiagaraDataInterfaceSkeletalMesh]
 usedDynamicInputs []
 usedModules [
  &self:SystemSpawnScript.SystemScriptSource.SystemScriptGraph.NiagaraNodeAssignment_0.SetVariables_#1
  &SystemState
 ]
}
emitters {
 BodySketches {
  emitterUpdateScript [
   &EmitterState()
   &SpawnRate(SpawnRate=2.5, Use Spawn Probability=false)
  ]
  particleSpawnScript [
   &PartitionParticles()
   &InitializeParticle(
    Sprite Rotation Mode=::NewEnumerator2
    Sprite Size Mode=::NewEnumerator3
    Uniform Sprite Size=&Multiply_Float(
     A=&Multiply_Float(
      A=&RandomRangeFloat(Minimum=45, Maximum=60)
      B=&SelectFloatFromArray(
       Direct Array Index=&MakeIntFromFloat(Float=$User.VariantID)
       Float Selection Array=DataInterface() {propertyValues {FloatData [1,0.7]}}
      )
     )
     B=&FN_FrontEndFloatModifier(Gameplay=1, Front End=0.82)
    )
    UsePositionOffset=false
    Write Material Random=true
   )
  ]
  particleUpdateScript [
   &ParticleState()
   &SkeletalMeshLocation(
    Bone Sampling Mode=::NewEnumerator4
    MaskSpawnGroup=false
    Sampled Position Offset=&MakeCustomVectorFromBool(Bool=&SetBoolByFloatComparison(A=$User.VariantID, B=0), True Vector=<0 0 -8>, False Vector=<1 0 0>)
    Skeletal Mesh=$System.Skeletal Mesh Reader
   )
   &CameraOffset(Camera Offset Amount=5)
   &Color(Color=&MultiplyLinearColor(
     Linear Color=&Multiply_LinearColorByFloat(Linear Color=$User.Base Color, Float=&FN_FrontEndFloatModifier(Gameplay=1, Front End=2))
     Scale Factor=&MakeVector4_FromLinearColor(LinearColor=&FNTimeOfDayColor(Night Color=<rgba 0.1334 0.1383 0.2726 1>, Morning Color=<rgba 1 0.6954 0.6539 1>, Day Color=<rgba 1 1 1 1>, Evening Color=<rgba 1 1 1 1>))
    ))
  ]
  rendererClasses [&NiagaraSpriteRendererProperties]
  renderers [{rendererClass=&NiagaraSpriteRendererProperties}]
 }
 EyeGlow {
  emitterUpdateScript [
   &EmitterState(Life Cycle Mode=::NewEnumerator1, Loop Behavior=::NewEnumerator1, Loop Duration=1, UseLoopDelay=false)
   &SpawnBurst_Instantaneous(Spawn Count=2, Use Spawn Probability=false, UseLoopCountLimit=false)
  ]
  particleSpawnScript [&InitializeParticle(Sprite Size Mode=::NewEnumerator3, Uniform Sprite Size=16, UsePositionOffset=false, Write Material Random=true)]
  particleUpdateScript [
   &ParticleState(Kill Particles When Lifetime Has Elapsed=false)
   &SkeletalMeshLocation(
    Bone / Socket Index=&ReturnExecIndex()
    MaskSpawnGroup=false
    Mesh Sampling Type=::NewEnumerator5
    Skeletal Mesh=$System.Skeletal Mesh Reader
    Socket Sampling Mode=::NewEnumerator5
   )
   &CameraOffset(Camera Offset Amount=4)
   &Color(Color=&MultiplyLinearColor(Linear Color=$User.EmissiveTint, Scale Factor=<1 1 1 0.5>))
  ]
  rendererClasses [&NiagaraSpriteRendererProperties]
  renderers [{rendererClass=&NiagaraSpriteRendererProperties}]
 }
}
summary {
 systemName=NS_Companion_DustyWing_Idle
 emitters [
  {emitterName=EyeGlow, rendererClasses [&NiagaraSpriteRendererProperties]}
  {emitterName=BodySketches, rendererClasses [&NiagaraSpriteRendererProperties]}
 ]
 userVariables [
  User.Base Color {type=LinearColor, description="", defaultValue {struct=LinearColor, value=<rgba 0 0 0 1>}}
  User.bIsFrontEnd {type=NiagaraBool}
  User.bIsFrontEndPreview {type=NiagaraBool}
  User.EmissiveTint {type=LinearColor, description="", defaultValue {struct=LinearColor, value=<rgba 1 0.0063 0 1>}}
  User.VariantID {type=NiagaraFloat, defaultValue=0, description=""}
 ]
}
