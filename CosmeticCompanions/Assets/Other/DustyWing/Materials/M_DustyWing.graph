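# Skinned-mesh material for the DustyWing cosmetic companion, built on the 2023 character parent functions.
# Pre-FX customization, companion custom color and basic tech art feed DustyWing's PBR/toon mix and body functions, ending in shared post-FX.
# Before post-FX, a view-angle blend of facing and rim terms mixes tint with dimmed emissive, applies a power, and replaces emissive where the emissive texture lies outside the tech-art red and green masks.
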Material /CosmeticCompanions/Assets/Other/DustyWing/Materials/M_DustyWing
graph {
 properties {
  bAutomaticallySetUsageInEditor=false
  bUseMaterialAttributes=true
  bUsedWithMeshDeformer=true
  bUsedWithSkeletalMesh=true
  bUsedWithStaticLighting=true
  shadingModel=MSM_FromMaterialExpression
  usedShadingModels=Default Lit | Subsurface | Clear Coat | Subsurface Profile | Cloth
 }
 nodes [
  n1 = MaterialFunctionCall!(Overide Texture Set UVs, Override Tech Art Mask UVs, Override Basic Emissive UVs) {materialFunction=/Game/Global/Materials/Character_Parent_2023/Functions/MF_CharacterParent_2023_Customization_PreFX}
  n2 = MaterialFunctionCall!(
   Tech Art Mask=@n1.Tech Art Mask
   Emissive UVs=@n1.Tech Art Mask
   LocalPosition=@n1.Tech Art Mask
   Fresnel=@n1.Tech Art Mask
   Mask Input=@n1.Tech Art Mask
   Material=MaterialFunctionCall!(Material=@n1.Result, Tech Art Mask=@n1.Result) {materialFunction=/CosmeticCompanions/Materials/MaterialFunctions/MF_FN_Companion_Custom_Color}.Base Color
  ) {materialFunction=/Game/Global/Materials/Character_Parent_2023/Functions/MF_CharacterParent_2023_CustomizationBasicTechArt}
  n3 = GetMaterialAttributes(None=MaterialFunctionCall!(In_MaterialAttributes=MaterialFunctionCall!(MA_In=@n2.Result, TechArtMask=@n1.Tech Art Mask, M=@n1.Tech Art Mask) {materialFunction=&/MF_DustyWing_Body_PBRToonMix}.MA_Out) {materialFunction=&/MF_DustyWing_Body}.Result) {attributeGetTypes [#1]}
  n4 = DotProduct(A=CameraVectorWS(), B=VertexNormalWS())
  n5 = Power(Base=@n4) {constExponent=4}
  n6 = Power(Base=OneMinus(None=@n4)) {constExponent=8}
  n7 = Power(Base=LinearInterpolate(
    A=VectorParameter() {group=06_Emissive, parameterName=EmissiveTint2, defaultValue=<rgba 0 0 0 1>}.RGB
    B=Multiply(A=@n3.Emissive Color) {constB=0.5}
    Alpha=Max(A=@n5, B=@n6)
   ))
  n8 = ComponentMask(None=@n1.Tech Art Mask) {r=true}
  n9 = ComponentMask(None=@n1.Tech Art Mask) {g=true}
  n10 = Multiply(A=OneMinus(None=Add(A=@n8, B=@n9)), B=TextureSample() {texture=/CosmeticCompanions/Assets/Other/DustyWing/Textures/T_DustyWing_Body_E}.R)
  MaterialFunctionCall(
   Material=SetMaterialAttributes(MaterialAttributes=@n3.MaterialAttributes, Emissive Color=LinearInterpolate(A=@n3.Emissive Color, B=@n7, Alpha=@n10)) {attributeSetTypes [#1]}
   M=@n1.M
   Pre Skinned Local Position=@n1.M
   Distance Blend=@n1.M
   Fresnel=@n1.M
  ) {materialFunction=&MF_CharacterParent_2023_PostFX}
 ]
}
