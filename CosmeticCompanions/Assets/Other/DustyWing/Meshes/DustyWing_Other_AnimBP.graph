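# Animation Blueprint for the DustyWing cosmetic companion that retargets a source mesh's pose onto its skeleton.
# A bCustomEmote bool selects the emote IK Retargeter when true, or the standard retargeter when false.
# The standard path layers a WingFlapA animation onto the left and right fin base bones through a per-bone blend.
# A Copy Pose From Mesh node sits unconnected, and the event graph is empty.
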AnimBlueprint /CosmeticCompanions/Assets/Other/DustyWing/Meshes/DustyWing_Other_AnimBP
target_skeleton=&/DustyWing_Other_Skeleton
graphs [
 AnimGraph {nodes [
   Copy Pose From Mesh() {bCopyCurves=true, bUseAttachedParent=true}
   n1 = Retarget Pose From Mesh!() -> (Pose:Pose Link Structure) {is=title AnimGraphNode_RetargetPoseFromMesh, iKRetargeterAsset=&/RTG_DustyWing}
   n2 = Sequence Player 'DustyWing_WingFlapA'() {sequence=/CosmeticCompanions/Assets/Other/DustyWing/Animation/Actions/DustyWing_WingFlapA}
   Output Pose(Result=Blend Poses by bool(
     BlendPose_0=Retarget Pose From Mesh() {iKRetargeterAsset=&/RTG_DustyWing_Emote}
     BlendPose_1=Layered blend per bone(BasePose=@n1, BlendPoses_0=@n2) {layerSetup [{branchFilters [{boneName=fin_base_l_01}, {boneName=fin_base_r_01}]}]}
     bActiveValue=Get bCustomEmote() -> (bCustomEmote:bool)
    ))
  ]}
 EventGraph {links [], nodes []}
]
variables []
